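# Idle Niagara effect for the HammerVice backpack cosmetic, adding warm sparkles, glitter and a glow that scale with its reactivity state.
# Sparkles spawns continuously at a rate driven by ReactivityState, while Glitter, Sparkles_React and Light burst on the Reactivity bool, all gated off at certain states.
# Particles sample backpack sockets, with partition-indexed offsets placing the sparks, and Light adds a light renderer.
# Front-end modifiers retune size, lifetime and motion, and time-of-day color scales brightness.
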NiagaraSystem /BRCosmetics/Accessories/FORT_Backpacks/Backpack_F_MED_HammerVice/FX/NS_Backpack_HammerVice_Idle
dependencies {
 usedDataInterfaces [
  &NiagaraDataInterfaceMeshRendererInfo
  &NiagaraDataInterfaceCurve
  &NiagaraDataInterfaceArrayFloat
  &NiagaraDataInterfaceArrayInt32
  &NiagaraDataInterfaceArrayFloat2
  &NiagaraDataInterfaceSocketReader
  &NiagaraDataInterfaceColorCurve
 ]
 usedDynamicInputs [&ColorFromCurve]
 usedModules [
  &self:SystemSpawnScript.SystemScriptSource.SystemScriptGraph.NiagaraNodeAssignment_7.SetVariables_#1
  &SystemState
  &self:SystemSpawnScript.SystemScriptSource.SystemScriptGraph.NiagaraNodeAssignment_1.SetVariables_#2
 ]
}
emitters {
 Glitter {
  emitterUpdateScript [
   &EmitterState()
   /Game/Effects/Niagara/Modules/Spawn/FN_Spawn_TriggeredBurst(
    Cooldown=0
    Module.Use Burst Count Limit=false
    Spawn:NiagaraBool=$User.Reactivity
    Spawn Count:NiagaraInt32=&Multiply_Int(
     A=&Multiply_Int(A=1, B=$User.ReactivityState)
     B=&MakeCustomIntFromBool(
      Bool=&Boolean_OrOperation(
       Bool1=&SetBoolByIntComparison(A=$User.ReactivityState, B=0, Comparison Type=::NewEnumerator4)
       Bool2=&SetBoolByIntComparison(A=$User.ReactivityState, B=5, Comparison Type=::NewEnumerator2)
      )
      True INT=0
      False INT=1
     )
    )
    UseLoopCountLimit=false
   )
  ]
  particleSpawnScript [
   &PartitionParticles(Number Of Partitions=$User.ReactivityState)
   &InitializeParticle(
    Color=<rgba 4 2.1333 0.8 1>
    Lifetime=&FN_FrontEndFloatModifier(
     Gameplay=&RandomRangeFloat(Minimum=0.3, Maximum=0.4)
     Front End=&RandomRangeFloat(Minimum=0.75, Maximum=1.25)
    )
    Sprite Size Mode=::NewEnumerator3
    Uniform Sprite Size=&FN_FrontEndFloatModifier(
     Gameplay=&RandomRangeFloat(Minimum=0.8, Maximum=1.5)
     Front End=&RandomRangeFloat(Minimum=0.4, Maximum=0.8)
    )
    UsePositionOffset=false
    Write Material Random=true
   )
   &InheritVelocity(
    Inherited Velocity Speed Limit:NiagaraFloat=&RandomRangeFloat(Minimum=40, Maximum=80)
    Speed Limit=true
    UseSpeedThreshold=false
   )
   &SocketLocation(
    SamplePosOffset=true
    Sampled Position Offset=&MergeHelper_VectorSwitch(
     0=<2.5 -1.5 0>
     1=<1 -1 -0.5>
     2=<0 0 0>
     3=<3 3 1>
     4=<0 0 0>
     5=<-1 -3.5 0>
     Index=$Particles.PartitionParticles.Partition
     Num=::NewEnumerator4
    )
    Socket Reader=$System.Socket Reader
    Socket Sampling Mode=::NewEnumerator3
    VelocitySpeedLimit=false
   )
   &ShapeLocation(MaskSpawnGroup=false, Non Uniform Scale=<10 1 1>, Override Seed=false, Sphere Radius=1, UseOwnerScale=true)
   &AddVelocity(
    Cone Angle=60
    Cone Axis=<0 0 1>
    Inner Cone Angle=30
    Module.UseConeFalloff=false
    Module.UseVelDistribution=false
    Override Seed=false
    Velocity Mode=::NewEnumerator2
    Velocity Speed=&Multiply_Float(
     A=&RandomRangeFloat(Minimum=50, Maximum=150)
     B=&FN_FrontEndFloatModifier(Gameplay=1, Front End=0.5)
    )
   )
  ]
  particleUpdateScript [
   &ParticleState()
   &ScaleSpriteSize(Scale Sprite Size Mode=::NewEnumerator2) {module=ScaleSpriteSize001}
   &ScaleSpriteSize(
    Uniform Curve Scale=&FN_FrontEndFloatModifier(Gameplay=1, Front End=0.75)
    Uniform Curve Sprite Scale=DataInterface() {propertyValues {Curve {keys [
        {tangentMode=RCTM_User, time=0, value=1}
        {arriveTangent=-0.0405, interpMode=RCIM_Cubic, leaveTangent=-0.0405, tangentMode=RCTM_User, time=0.2, value=1}
        {arriveTangent=-2.1547, interpMode=RCIM_Cubic, leaveTangent=-2.1547, tangentMode=RCTM_User, time=1, value=0}
       ]}}}
   )
   &Drag(
    Drag=&FN_FrontEndFloatModifier(
     Gameplay=&RandomRangeFloat(Minimum=2.5, Maximum=5)
     Front End=&RandomRangeFloat(Minimum=6, Maximum=8)
    )
    Use Linear Drag=true
    Use Rotational Drag=true
   )
   &SolveForcesAndVelocity(Clamp Velocity=false, Limit Acceleration=false)
   /SetVariables(
    Particles.Color2=&Multiply_LinearColorByFloat(
     Linear Color=<rgba 0.1 0.0383 0 1>
     Float=&FN_Select_Float_from_Time_Of_Day(Morning=0.8, Day=0.8, Evening=0.8, Night=0.8)
    )
    Particles.SpriteSize2=&Vector2DFromFloat(Value=&FN_FrontEndFloatModifier(
      Gameplay=&RandomRangeFloat(Minimum=8, Maximum=10)
      Front End=&RandomRangeFloat(Minimum=6, Maximum=4)
     ))
   ) {isSetParameters=true}
   &ScaleColor(
    Scale Alpha=&FN_FrontEndFloatModifier(Gameplay=1, Front End=0.5)
    Scale RGB=&VectorFromFloat(Value=&FN_FrontEndFloatModifier(Gameplay=&RandomRangeFloat(Minimum=-10, Maximum=10, Evaluation Type=::NewEnumerator1), Front End=1))
    ScaleA=true
    ScaleRGB=true
   )
   &ScaleColor(
    Scale Alpha=&FNTimeOfDayScalar(Night Scale=1, Morning Scale=1, Evening Scale=1)
    Scale RGB=&MakeVectorFromLinearColorRGB(Color=$System.TOD)
    ScaleA=true
    ScaleRGB=true
   ) {module=ScaleColor001}
  ]
  rendererClasses [&NiagaraSpriteRendererProperties]
  renderers [
   {rendererClass=&NiagaraSpriteRendererProperties}
   {rendererClass=&NiagaraSpriteRendererProperties, rendererIndex=1}
  ]
 }
 Light {
  emitterUpdateScript [
   &EmitterState()
   &FN_Spawn_TriggeredBurst(
    Cooldown=0
    Module.Use Burst Count Limit=false
    Spawn:NiagaraBool=$User.Reactivity
    Spawn Count:NiagaraInt32=&Multiply_Int(
     A=5
     B=&MakeCustomIntFromBool(
      Bool=&SetBoolByIntComparison(A=$User.ReactivityState, B=0, Comparison Type=::NewEnumerator4)
      True INT=0
      False INT=1
     )
    )
    UseLoopCountLimit=false
   )
   &FN_Spawn_TriggeredBurst(
    Burst Count Limit=1
    Cooldown=0
    Module.Use Burst Count Limit=true
    Spawn:NiagaraBool=$User.Reactivity
    Spawn Count:NiagaraInt32=&MakeCustomIntFromBool(
     Bool=&SetBoolByIntComparison(A=$User.ReactivityState, B=0, Comparison Type=::NewEnumerator4)
     True INT=0
     False INT=1
    )
    Spawn Group=1
    UseLoopCountLimit=false
   ) {module=FN_Spawn_TriggeredBurst001}
  ]
  particleSpawnScript [
   /SetVariables(Particles.VisibilityTag=$Emitter.SpawnGroup) {isSetParameters=true}
   &InitializeParticle(
    Color Maximum=<rgba 0.8333 0.4919 0.0798 0.5>
    Color Minimum=<rgba 0.8333 0.3827 0.0608 0.5>
    Color Mode=::NewEnumerator2
    Lifetime Max=0.75
    Lifetime Min=0.5
    Lifetime Mode=::NewEnumerator1
    Mesh Scale Mode=::NewEnumerator3
    Mesh Uniform Scale=0.2
    Position Mode=::NewEnumerator1
    Sprite Size=&FN_FrontEndVector2Modifier(
     Front End:Vector2f=&RandomRangeVector2D(Minimum=<6 10>, Maximum=<8 20>)
     Gameplay:Vector2f=&RandomRangeVector2D(Minimum=<6 20>, Maximum=<8 25>)
    )
    Sprite Size Mode=::NewEnumerator4
    Sprite UV Mode=::NewEnumerator1
    UsePositionOffset=false
    Write Material Random=true
   )
   &DynamicMaterialParameters(
    Index 0 Param 1=&RandomRangeFloat(Minimum=-0.2, Maximum=0.2)
    Index 0 Param 2=1
    Index 0 Param 3=0.5
    Module.Param0WriteEnabled=true
    Module.Param1WriteEnabled=false
    Module.Param2WriteEnabled=false
    Module.Param3WriteEnabled=false
   )
   &SocketLocation(
    SamplePosOffset=true
    Sampled Position Offset=<0 -2.5 1>
    Socket Reader=$System.Socket Reader
    Socket Sampling Mode=::NewEnumerator3
    VelocitySpeedLimit=false
   )
   &ShapeLocation(
    Handle Radius=0
    Large Radius=5
    MaskSpawnGroup=false
    Non Uniform Scale=<1.5 0.2 1>
    Override Seed=false
    Rotation Axis=<0 1 0>
    Rotation Mode=::NewEnumerator1
    Shape Primitive=::NewEnumerator3
    U Distribution=0.5
    UseOwnerScale=true
   )
   &AddVelocity(Module.Use Falloff Distance=false, Origin Offset=<0 -2 -15>, Override Seed=false, Velocity Mode=::NewEnumerator1, Velocity Speed=1)
  ]
  particleUpdateScript [
   &ParticleState()
   &Drag(Drag=5, Use Linear Drag=true, Use Rotational Drag=false)
   &SolveForcesAndVelocity(Clamp Velocity=false, Limit Acceleration=false)
   &ScaleColor(
    Scale Alpha=&FloatFromCurve(
     FloatCurve=DataInterface() {propertyValues {Curve {keys [
         {arriveTangent=6.668, interpMode=RCIM_Cubic, leaveTangent=6.668, tangentMode=RCTM_User, time=0, value=0}
         {arriveTangent=0.979, interpMode=RCIM_Cubic, leaveTangent=0.979, time=0.0895, value=1.0296}
         {arriveTangent=-2.8295, interpMode=RCIM_Cubic, leaveTangent=-2.8295, tangentMode=RCTM_User, time=0.4911, value=0.4808}
         <at 1 0>
        ]}}}
     Scale Curve=&FN_FrontEndFloatModifier(Gameplay=1, Front End=0.5)
    )
    Scale RGB=&Multiply_VectorByFloat(Vector=&MakeVectorFromLinearColorRGB(Color=$System.TOD), Float=5)
    ScaleA=true
    ScaleRGB=true
   )
   &ScaleSpriteSize(
    Uniform Curve Scale=&FN_FrontEndFloatModifier(Gameplay=1, Front End=0.75)
    Uniform Curve Sprite Scale=DataInterface() {propertyValues {Curve {keys [
        {interpMode=RCIM_Cubic, tangentMode=RCTM_User, time=0, value=0}
        {arriveTangent=-0.427, interpMode=RCIM_Cubic, leaveTangent=-0.427, tangentMode=RCTM_User, time=1, value=1.525}
       ]}}}
   )
   &SpriteRotationRate(Rotation Rate=50)
   &FN_Light_Attributes(
    Color=<rgba 0.8333 0.4919 0.0798 1>
    Create Color Attribute=true
    Create Exponent Attribute=true
    Create Radius Attribute=true
    Exponent=5
    Radius=20
    Scale Color=7.5
   )
  ]
  rendererClasses [&NiagaraSpriteRendererProperties, &NiagaraLightRendererProperties]
  renderers [
   {rendererClass=&NiagaraSpriteRendererProperties}
   {rendererClass=&NiagaraLightRendererProperties, rendererIndex=1}
  ]
 }
 Sparkles {
  emitterUpdateScript [
   &EmitterState()
   &SpawnRate(
    Spawn Probability=&MakeFloatFromInt(Int=&MakeCustomIntFromBool(
      Bool=&Boolean_OrOperation(
       Bool1=&SetBoolByIntComparison(A=$User.ReactivityState, B=0, Comparison Type=::NewEnumerator4)
       Bool2=&SetBoolByIntComparison(A=$User.ReactivityState, B=5, Comparison Type=::NewEnumerator2)
      )
      True INT=0
      False INT=1
     ))
    SpawnRate=&Multiply_FloatByInt(Float=0.33, Integer=$User.ReactivityState)
    Use Spawn Probability=true
   )
  ]
  particleSpawnScript [
   &PartitionParticles(Number Of Partitions=&MergeHelper_IntSwitch(0=0, 1=1, 2=3, 3=5, 4=6, 5=?, Index=$User.ReactivityState, Num=::NewEnumerator3))
   &InitializeParticle(
    Color=&MergeHelper_ColorSwitch(
     0=<rgba 3.3588 4 3.9618 1>
     1=<rgba 4 2.1333 0.8 1>
     2=<rgba 4 2.1333 0.8 1>
     3=<rgba 4 2.1333 0.8 1>
     4=<rgba 4 2.1333 0.8 1>
     5=<rgba 3.3588 4 3.9618 1>
     Index=$Particles.PartitionParticles.Partition
     Num=::NewEnumerator4
    )
    Lifetime Max=1
    Lifetime Min=0.3
    Lifetime Mode=::NewEnumerator1
    Sprite Size Mode=::NewEnumerator1
    Uniform Sprite Size Max=15
    Uniform Sprite Size Min=13
    UsePositionOffset=false
    Write Material Random=true
   )
   &ApplyInitialForces()
   &SocketLocation(
    SamplePosOffset=true
    Sampled Position Offset=&MergeHelper_VectorSwitch(
     0=<2.5 -1.5 0>
     1=<1 -1 -0.5>
     2=<0 0 0>
     3=<3 3 1>
     4=<0 0 0>
     5=<-1 -3.5 0>
     Index=$Particles.PartitionParticles.Partition
     Num=::NewEnumerator4
    )
    Socket Index=&Add_Integer(A=$Particles.PartitionParticles.Partition, B=1)
    Socket Reader=$System.Socket Reader
    Socket Sampling Mode=::NewEnumerator3
    VelocitySpeedLimit=false
   )
   &ShapeLocation(
    MaskSpawnGroup=false
    Non Uniform Scale=&MergeHelper_VectorSwitch(
     0=<1 1 1.5>
     1=<1 1 1.25>
     2=<1 1 7.5>
     3=<2 0.75 1>
     4=<2 1 2>
     5=<2.5 1 8>
     Index=$Particles.PartitionParticles.Partition
     Num=::NewEnumerator4
    )
    Override Seed=false
    Sphere Radius=&MergeHelper_FloatSwitch(0=2.5, 1=2.5, 2=1, 3=5, 4=5, 5=1, Index=$Particles.PartitionParticles.Partition, Num=::NewEnumerator4)
    UseOwnerScale=true
   )
   &CameraOffset(Camera Offset Amount=10)
  ]
  particleUpdateScript [
   &ParticleState()
   &ScaleSpriteSize(
    Uniform Curve Index=&Add_Float(
     A=&Multiply_Float(A=$Particles.Age, B=&RandomRangeFloat(Minimum=0.8, Maximum=1.5))
     B=&RandomRangeFloat(Minimum=-0.25, Maximum=0.25)
    )
    Uniform Curve Sprite Scale=DataInterface() {propertyValues {Curve {keys [
        {tangentMode=RCTM_User, time=0.2, value=0}
        <at 0.22 1 RCIM_Cubic>
        {tangentMode=RCTM_User, time=0.5, value=0}
        <at 0.68 0>
        <at 0.7 0.8 RCIM_Cubic>
        <at 0.85 0>
       ]}}}
   )
   &Drag(Drag=2, Use Linear Drag=true, Use Rotational Drag=false)
   &SolveForcesAndVelocity(Clamp Velocity=false, Limit Acceleration=false)
   &ScaleColor(
    Scale Alpha=&FloatFromCurve(FloatCurve=DataInterface() {propertyValues {Curve {keys [<at 0.5 1 RCIM_Cubic>,<at 1 0>]}}})
    Scale RGB=&Multiply_Vector(
     A=&VectorFromFloat(Value=&FloatFromCurve(
       FloatCurve=DataInterface() {propertyValues {Curve {keys [{arriveTangent=-13.9762, interpMode=RCIM_Cubic, leaveTangent=-13.9762, tangentMode=RCTM_User, time=0, value=5}, <at 0.5 1>]}}}
       Scale Curve=0.5
      ))
     B=&MakeVectorFromLinearColorRGB(Color=$System.TOD)
    )
    ScaleA=true
    ScaleRGB=true
   )
  ]
  rendererClasses [&NiagaraSpriteRendererProperties]
  renderers [{rendererClass=&NiagaraSpriteRendererProperties}]
 }
 Sparkles_React {
  emitterUpdateScript [
   &EmitterState(
    Enable Distance Culling=true
    Enable Visibility Culling=true
    MaxDistance=2500
    Module.bApplySpawnCountScaleByDistance=true
    Module.bUseMaxDistance=true
    Module.bUseMinDistance=false
    Scalability Mode=::NewEnumerator1
    Scale Spawn Count=true
   )
   &FN_Spawn_TriggeredBurst(
    Cooldown=0
    Module.Use Burst Count Limit=false
    Spawn:NiagaraBool=$User.Reactivity
    Spawn Count:NiagaraInt32=&Multiply_Int(
     A=&Multiply_Int(A=2, B=$User.ReactivityState)
     B=&MakeCustomIntFromBool(
      Bool=&Boolean_OrOperation(
       Bool1=&SetBoolByIntComparison(A=$User.ReactivityState, B=0, Comparison Type=::NewEnumerator4)
       Bool2=&SetBoolByIntComparison(A=$User.ReactivityState, B=5, Comparison Type=::NewEnumerator2)
      )
      True INT=0
      False INT=1
     )
    )
    UseLoopCountLimit=false
   )
   &FN_Spawn_TriggeredBurst(
    Cooldown=0
    Module.Use Burst Count Limit=false
    Spawn:NiagaraBool=$User.Reactivity
    Spawn Count:NiagaraInt32=&Multiply_Int(
     A=&Multiply_Int(A=1, B=$User.ReactivityState)
     B=&MakeCustomIntFromBool(
      Bool=&Boolean_OrOperation(
       Bool1=&SetBoolByIntComparison(A=$User.ReactivityState, B=0, Comparison Type=::NewEnumerator4)
       Bool2=&SetBoolByIntComparison(A=$User.ReactivityState, B=5, Comparison Type=::NewEnumerator2)
      )
      True INT=0
      False INT=1
     )
    )
    Spawn Group=1
    UseLoopCountLimit=false
   ) {module=FN_Spawn_TriggeredBurst001}
  ]
  particleSpawnScript [
   &PartitionParticles(Number Of Partitions=$User.ReactivityState)
   &InitializeParticle(
    Color=<rgba 4 2.1333 0.8 1>
    Lifetime=&FN_FrontEndFloatModifier(
     Gameplay=&RandomRangeFloat(Minimum=0.3, Maximum=0.2)
     Front End=&RandomRangeFloat(Minimum=0.75, Maximum=1)
    )
    Sprite Rotation Mode=::NewEnumerator2
    Sprite Size Mode=::NewEnumerator3
    Uniform Sprite Size=&FN_FrontEndFloatModifier(
     Gameplay=&MergeHelper_FloatSwitch(
      0=&RandomRangeFloat(Minimum=1.5, Maximum=8)
      1=&RandomRangeFloat(Minimum=7.5, Maximum=10)
      2=?
      3=?
      Index=&SpawnGroup()
      Num=::NewEnumerator0
     )
     Front End=&RandomRangeFloat(Minimum=1.5, Maximum=8)
    )
    UsePositionOffset=true
    Write Material Random=true
   )
   &SocketLocation(
    SamplePosOffset=true
    Sampled Position Offset=&MergeHelper_VectorSwitch(
     0=<2.5 -1.5 0>
     1=<1 -1 -0.5>
     2=<0 0 0>
     3=<3 3 1>
     4=<0 0 0>
     5=<-1 -3.5 0>
     Index=$Particles.PartitionParticles.Partition
     Num=::NewEnumerator4
    )
    Socket Reader=$System.Socket Reader
    Socket Sampling Mode=::NewEnumerator3
    VelocitySpeedLimit=false
   )
   &ShapeLocation(MaskSpawnGroup=false, Non Uniform Scale=<10 1 1>, Override Seed=false, Sphere Radius=1, UseOwnerScale=true)
   &AddVelocity(
    Cone Angle=60
    Cone Axis=<0 0 1>
    Inner Cone Angle=30
    Module.UseConeFalloff=false
    Module.UseVelDistribution=false
    Override Seed=false
    Velocity Mode=::NewEnumerator2
    Velocity Speed=&Multiply_Float(
     A=&RandomRangeFloat(Minimum=50, Maximum=150)
     B=&FN_FrontEndFloatModifier(Gameplay=1, Front End=0.75)
    )
   )
  ]
  particleUpdateScript [
   &ParticleState()
   &SpriteRotationRate(Rotation Rate=&RandomRangeFloat(Minimum=-90, Maximum=90))
   &Drag(
    Drag=&FN_FrontEndFloatModifier(Gameplay=5, Front End=8)
    Use Linear Drag=true
    Use Rotational Drag=true
   )
   &SolveForcesAndVelocity(Clamp Velocity=false, Limit Acceleration=false)
   &ScaleSpriteSize(Uniform Curve Sprite Scale=DataInterface() {propertyValues {Curve {
       defaultValue=0
       keys [{interpMode=RCIM_Cubic, tangentMode=RCTM_User, time=0, value=1}, {interpMode=RCIM_Cubic, tangentMode=RCTM_User, time=1, value=0}]
      }}})
   &ScaleColor(
    Scale Alpha=&FN_FrontEndFloatModifier(Gameplay=&FNTimeOfDayScalar(Night Scale=1, Morning Scale=1, Evening Scale=1), Front End=0.5)
    Scale RGB=&MakeVectorFromLinearColorRGB(Color=$System.TOD)
    ScaleA=true
    ScaleRGB=true
   )
  ]
  rendererClasses [&NiagaraSpriteRendererProperties]
  renderers [{rendererClass=&NiagaraSpriteRendererProperties}]
 }
}
summary {
 systemName=NS_Backpack_HammerVice_Idle
 emitters [
  {emitterName=Sparkles, rendererClasses [&NiagaraSpriteRendererProperties]}
  {emitterName=Glitter, rendererClasses [&NiagaraSpriteRendererProperties]}
  {emitterName=Sparkles_React, rendererClasses [&NiagaraSpriteRendererProperties]}
  {emitterName=Light, rendererClasses [&NiagaraSpriteRendererProperties, &NiagaraLightRendererProperties]}
 ]
 userVariables [
  User.bIsFrontEnd {type=NiagaraBool}
  User.bIsFrontEndPreview {type=NiagaraBool}
  User.Reactivity {type=NiagaraBool}
  User.ReactivityState {type=NiagaraInt32}
 ]
}
