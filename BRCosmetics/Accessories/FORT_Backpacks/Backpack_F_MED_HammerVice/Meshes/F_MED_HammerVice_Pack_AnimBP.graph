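# Anim blueprint for the HammerVice backpack, a reactive cosmetic that gains visual stages as chests are opened.
# Chest-opened anim events raise a capped state counter, which auto-cycles in front-end or test previews and is broadcast each update.
# The state drives "Hide Element" material scalars and bool-blended bone transforms on gold bars; IsAlwaysVisible forces all elements shown.
# AnimDynamics chains swing necklaces, bag, cat, glasses and stars, with one necklace rotation-constrained to another.
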AnimBlueprint /BRCosmetics/Accessories/FORT_Backpacks/Backpack_F_MED_HammerVice/Meshes/F_MED_HammerVice_Pack_AnimBP
target_skeleton=&/F_MED_HammerVice_Pack_Skeleton
graphs [
 AnimGraph {nodes [
   n1 = Input Pose()
   n2 = Get GoldenBarScale4() -> (GoldenBarScale4:bool)
   n3 = "Transform (Modify) Bone - Bone: GoldBar_stp4_1_jnt"!(
    ComponentPose="Transform (Modify) Bone - Bone: pack"(ComponentPose=Local To Component(LocalPose=@n1), Translation="0,-1,0", Rotation="0,0,0") {translationMode=BMM_Additive, boneToModify {boneName=pack}, translation {y=-1}}
    bAlphaBoolEnabled:bool=@n2
    Translation="1,1,2"
    Rotation="0,0,0"
    Scale="0.1,1,1"
   ) -> (Pose) {
    alphaInputType=Bool
    scaleMode=BMM_Additive
    alphaBoolBlend {
     blendOption=QuinticInOut
     blendOutTime=0.75
     customCurve=&/F_MED_HammerVice_Pack_curve
    }
    boneToModify {boneName=GoldBar_stp4_1_jnt}
    scale {x=0.1}
    translation=<1 1 2>
   }
   n4 = "Transform (Modify) Bone - Bone: GoldBar_stp4_2_jnt"!(ComponentPose=@n3, bAlphaBoolEnabled:bool=@n2, Translation="3,4,6", Rotation="0,0,0", Scale="0.1,1,1") -> (Pose) {
    alphaInputType=Bool
    scaleMode=BMM_Additive
    alphaBoolBlend {
     blendOption=Custom
     blendOutTime=0.5
     customCurve=&F_MED_HammerVice_Pack_curve
    }
    boneToModify {boneName=GoldBar_stp4_2_jnt}
    scale {x=0.1}
    translation=<3 4 6>
   }
   n5 = "Transform (Modify) Bone - Bone: GoldBar_stp4_3_jnt"!(ComponentPose=@n4, bAlphaBoolEnabled:bool=@n2, Translation="8,2,3", Rotation="0,0,0", Scale="0.1,1,1") -> (Pose) {
    alphaInputType=Bool
    scaleMode=BMM_Additive
    alphaBoolBlend {
     blendOption=Custom
     blendOutTime=1
     customCurve=&F_MED_HammerVice_Pack_curve
    }
    boneToModify {boneName=GoldBar_stp4_3_jnt}
    scale {x=0.1}
    translation=<8 2 3>
   }
   n6 = "Transform (Modify) Bone - Bone: GoldBar_stp4_4_jnt"!(ComponentPose=@n5, bAlphaBoolEnabled:bool=@n2, Translation="2,2,5", Rotation="0,0,0", Scale="0.1,1,1") -> (Pose) {
    alphaInputType=Bool
    scaleMode=BMM_Additive
    alphaBoolBlend {
     blendOption=CircularInOut
     blendOutTime=0.6
     customCurve=&F_MED_HammerVice_Pack_curve
    }
    boneToModify {boneName=GoldBar_stp4_4_jnt}
    scale {x=0.1}
    translation=<2 2 5>
   }
   n7 = Get GoldenBarScale5() -> (GoldenBarScale5:bool)
   n8 = "Transform (Modify) Bone - Bone: GoldBar_stp5_1_jnt"!(ComponentPose=@n6, bAlphaBoolEnabled:bool=@n7, Translation="10,8,5", Rotation="0,0,0", Scale="0.1,1,1") -> (Pose) {
    alphaInputType=Bool
    scaleMode=BMM_Additive
    alphaBoolBlend {
     blendOption=QuadraticInOut
     blendOutTime=0.75
     customCurve=&F_MED_HammerVice_Pack_curve
    }
    boneToModify {boneName=GoldBar_stp5_1_jnt}
    scale {x=0.1}
    translation=<10 8 5>
   }
   n9 = "Transform (Modify) Bone - Bone: GoldBar_stp5_2_jnt"!(ComponentPose=@n8, bAlphaBoolEnabled:bool=@n7, Translation="8,10,5", Rotation="0,0,0", Scale="0.1,1,1") -> (Pose) {
    alphaInputType=Bool
    scaleMode=BMM_Additive
    alphaBoolBlend {
     blendOption=Custom
     blendOutTime=1
     customCurve=&F_MED_HammerVice_Pack_curve
    }
    boneToModify {boneName=GoldBar_stp5_2_jnt}
    scale {x=0.1}
    translation=<8 10 5>
   }
   n10 = "Transform (Modify) Bone - Bone: GoldBar_stp5_3_jnt"!(ComponentPose=@n9, bAlphaBoolEnabled:bool=@n7, Translation="0,0,0", Rotation="0,0,0", Scale="0.1,1,1") -> (Pose) {
    alphaInputType=Bool
    scaleMode=BMM_Additive
    alphaBoolBlend {
     blendOption=Custom
     blendOutTime=0.75
     customCurve=&F_MED_HammerVice_Pack_curve
    }
    boneToModify {boneName=GoldBar_stp5_3_jnt}
    scale {x=0.1}
   }
   n11 = "Transform (Modify) Bone - Bone: GoldBar_stp5_4_jnt"!(ComponentPose=@n10, bAlphaBoolEnabled:bool=@n7, Translation="4,6,3", Rotation="0,0,0", Scale="0.1,1,1") -> (Pose) {
    alphaInputType=Bool
    scaleMode=BMM_Additive
    alphaBoolBlend {
     blendOption=Custom
     blendOutTime=0.5
     customCurve=&F_MED_HammerVice_Pack_curve
    }
    boneToModify {boneName=GoldBar_stp5_4_jnt}
    scale {x=0.1}
    translation=<4 6 3>
   }
   n12 = "Transform (Modify) Bone - Bone: GoldBar_stp5_5_jnt"!(ComponentPose=@n11, bAlphaBoolEnabled:bool=@n7, Translation="0,0,0", Rotation="0,0,0", Scale="0.1,1,1") -> (Pose) {
    alphaInputType=Bool
    scaleMode=BMM_Additive
    alphaBoolBlend {blendOption=QuinticInOut, blendOutTime=0.4}
    boneToModify {boneName=GoldBar_stp5_5_jnt}
    scale {x=0.1}
   }
   n13 = Anim Dynamics - Bone: Necklace_09_01_jnt(ComponentPose=@n12, Alpha=1, ExternalForce="0,0,0") {
    angularDampingOverride=0.999
    bOverrideAngularDamping=true
    bUseGravityOverride=true
    simulationSpace=RootRelative
    boundBone {boneName=Necklace_09_01_jnt}
    componentAppliedLinearAccClamp=<3000 3000 3000>
    componentLinearAccScale=<0.5 0.5 0.5>
    componentLinearVelScale=<0.3 0.3 0.3>
    gravityOverride {z=-980}
    physicsBodyDefinitions [{
      boundBone {boneName=Necklace_09_01_jnt}
      boxExtents=<10 10 10>
      constraintSetup {angularLimitsMax=<15 0 10>, angularLimitsMin=<-15 -15 -20>, angularTarget=<1 0 0>}
      localJointOffset=<1 0 0>
     }]
    simSpaceSettings {maxAngularAcceleration=25, maxAngularVelocity=5, simSpaceAngularAlpha=1}
   }
   n14 = Anim Dynamics - Bone: Necklace_09_02_jnt(ComponentPose=@n13, Alpha=1, ExternalForce="0,0,0") {
    angularDampingOverride=0.999
    bOverrideAngularDamping=true
    bUseGravityOverride=true
    simulationSpace=RootRelative
    boundBone {boneName=Necklace_09_02_jnt}
    componentAppliedLinearAccClamp=<3000 3000 3000>
    componentLinearAccScale=<0.5 0.5 0.5>
    componentLinearVelScale=<0.3 0.3 0.3>
    gravityOverride {z=-980}
    physicsBodyDefinitions [{
      boundBone {boneName=Necklace_09_02_jnt}
      boxExtents=<10 10 10>
      constraintSetup {angularLimitsMax=<15 0 10>, angularLimitsMin=<-15 -15 -20>, angularTarget=<1 0 0>}
      localJointOffset=<1 0 0>
     }]
    simSpaceSettings {maxAngularAcceleration=25, maxAngularVelocity=5, simSpaceAngularAlpha=1}
   }
   n15 = Anim Dynamics - Bone: Necklace_09_03_jnt(ComponentPose=@n14, Alpha=1, ExternalForce="0,0,0") {
    angularDampingOverride=0.999
    bOverrideAngularDamping=true
    bUseGravityOverride=true
    simulationSpace=RootRelative
    boundBone {boneName=Necklace_09_03_jnt}
    componentAppliedLinearAccClamp=<3000 3000 3000>
    componentLinearAccScale=<0.5 0.5 0.5>
    componentLinearVelScale=<0.3 0.3 0.3>
    gravityOverride {z=-980}
    physicsBodyDefinitions [{
      boundBone {boneName=Necklace_09_03_jnt}
      boxExtents=<10 10 10>
      constraintSetup {angularLimitsMax=<15 0 10>, angularLimitsMin=<-15 -15 -20>, angularTarget=<1 0 0>}
      localJointOffset=<1 0 0>
     }]
    simSpaceSettings {maxAngularAcceleration=25, maxAngularVelocity=5, simSpaceAngularAlpha=1}
   }
   n16 = Anim Dynamics - Bone: Necklace_07_01_jnt(ComponentPose=@n15, Alpha=1, ExternalForce="0,0,0") {
    angularDampingOverride=0.9999
    bOverrideAngularDamping=true
    bUseGravityOverride=true
    simulationSpace=RootRelative
    boundBone {boneName=Necklace_07_01_jnt}
    componentAppliedLinearAccClamp=<3000 3000 3000>
    componentLinearAccScale=<0.5 0.5 0.5>
    componentLinearVelScale=<0.3 0.3 0.3>
    gravityOverride {z=-980}
    physicsBodyDefinitions [{
      boundBone {boneName=Necklace_07_01_jnt}
      boxExtents=<10 10 10>
      constraintSetup {angularLimitsMax=<15 3 20>, angularLimitsMin=<-15 -20 -20>, angularTarget=<1 0 0>}
      localJointOffset=<1 0 0>
     }]
    simSpaceSettings {maxAngularAcceleration=25, maxAngularVelocity=5, simSpaceAngularAlpha=1}
   }
   n17 = Anim Dynamics - Bone: Necklace_07_02_jnt(ComponentPose=@n16, Alpha=1, ExternalForce="0,0,0") {
    angularDampingOverride=0.9999
    bOverrideAngularDamping=true
    bUseGravityOverride=true
    simulationSpace=RootRelative
    boundBone {boneName=Necklace_07_02_jnt}
    componentAppliedLinearAccClamp=<3000 3000 3000>
    componentLinearAccScale=<0.5 0.5 0.5>
    componentLinearVelScale=<0.3 0.3 0.3>
    gravityOverride {z=-980}
    physicsBodyDefinitions [{
      boundBone {boneName=Necklace_07_02_jnt}
      boxExtents=<10 10 10>
      constraintSetup {angularLimitsMax=<15 3 20>, angularLimitsMin=<-15 -20 -20>, angularTarget=<1 0 0>}
      localJointOffset=<1 0 0>
     }]
    simSpaceSettings {maxAngularAcceleration=25, maxAngularVelocity=5, simSpaceAngularAlpha=1}
   }
   n18 = Anim Dynamics - Bone: Necklace_07_03_jnt(ComponentPose=@n17, Alpha=1, ExternalForce="0,0,0") {
    angularDampingOverride=0.9999
    bOverrideAngularDamping=true
    bUseGravityOverride=true
    simulationSpace=RootRelative
    boundBone {boneName=Necklace_07_03_jnt}
    componentAppliedLinearAccClamp=<3000 3000 3000>
    componentLinearAccScale=<0.5 0.5 0.5>
    componentLinearVelScale=<0.3 0.3 0.3>
    gravityOverride {z=-980}
    physicsBodyDefinitions [{
      boundBone {boneName=Necklace_07_03_jnt}
      boxExtents=<10 10 10>
      constraintSetup {angularLimitsMax=<15 3 20>, angularLimitsMin=<-15 -20 -20>, angularTarget=<1 0 0>}
      localJointOffset=<1 0 0>
     }]
    simSpaceSettings {maxAngularAcceleration=25, maxAngularVelocity=5, simSpaceAngularAlpha=1}
   }
   n19 = Anim Dynamics - Bone: Necklace_06_01_jnt(ComponentPose=@n18, Alpha=1, ExternalForce="0,0,0") {
    angularDampingOverride=0.999
    bOverrideAngularDamping=true
    bUseGravityOverride=true
    simulationSpace=RootRelative
    boundBone {boneName=Necklace_06_01_jnt}
    componentAppliedLinearAccClamp=<3000 3000 3000>
    componentLinearAccScale=<0.3 0.3 0.3>
    componentLinearVelScale=<0.3 0.3 0.3>
    gravityOverride {z=-980}
    physicsBodyDefinitions [{
      boundBone {boneName=Necklace_06_01_jnt}
      boxExtents=<9 9 9>
      constraintSetup {angularLimitsMax=<5 3 20>, angularLimitsMin=<-5 -20 -20>, angularTarget=<1 0 0>}
      localJointOffset=<1 0 0>
     }]
    simSpaceSettings {maxAngularAcceleration=25, maxAngularVelocity=5, simSpaceAngularAlpha=1}
   }
   n20 = Anim Dynamics - Bone: Necklace_05_01_jnt(ComponentPose=@n19, Alpha=1, ExternalForce="0,0,0") {
    angularDampingOverride=0.999
    bOverrideAngularDamping=true
    bUseGravityOverride=true
    simulationSpace=RootRelative
    boundBone {boneName=Necklace_05_01_jnt}
    componentAppliedLinearAccClamp=<3000 3000 3000>
    componentLinearAccScale=<0.3 0.3 0.3>
    componentLinearVelScale=<0.3 0.3 0.3>
    gravityOverride {z=-980}
    physicsBodyDefinitions [{
      boundBone {boneName=Necklace_05_01_jnt}
      boxExtents=<10 10 10>
      constraintSetup {angularLimitsMax=<5 3 20>, angularLimitsMin=<-5 -20 -20>, angularTarget=<1 0 0>}
      localJointOffset=<1 0 0>
     }]
    simSpaceSettings {maxAngularAcceleration=25, maxAngularVelocity=5, simSpaceAngularAlpha=1}
   }
   n21 = Anim Dynamics - Bone: HammerVice_Bag_jnt(ComponentPose=@n20, Alpha=1, ExternalForce="0,0,0") {
    angularDampingOverride=1
    bGravityOverrideInSimSpace=true
    bOverrideAngularDamping=true
    bUseGravityOverride=true
    simulationSpace=RootRelative
    boundBone {boneName=HammerVice_Bag_jnt}
    componentAppliedLinearAccClamp=<3000 3000 3000>
    componentLinearAccScale=<0.3 0.3 0.3>
    componentLinearVelScale=<0.3 0.3 0.3>
    gravityOverride {z=980}
    physicsBodyDefinitions [{
      boundBone {boneName=HammerVice_Bag_jnt}
      boxExtents=<15 15 15>
      constraintSetup {angularLimitsMax=<15 20 15>, angularLimitsMin=<-15 -20 -15>, angularTarget=<1 0 0>}
      localJointOffset=<1 0 0>
     }]
    simSpaceSettings {maxAngularAcceleration=25, maxAngularVelocity=5, simSpaceAngularAlpha=1}
   }
   n22 = Anim Dynamics - Bone: Cat_jnt(ComponentPose=@n21, Alpha=1, ExternalForce="0,0,0") {
    angularDampingOverride=0.999
    bGravityOverrideInSimSpace=true
    bOverrideAngularDamping=true
    bUseGravityOverride=true
    simulationSpace=RootRelative
    boundBone {boneName=Cat_jnt}
    componentAppliedLinearAccClamp=<3000 3000 3000>
    componentLinearAccScale=<0.3 0.3 0.3>
    componentLinearVelScale=<0.3 0.3 0.3>
    gravityOverride {z=980}
    physicsBodyDefinitions [{
      boundBone {boneName=Cat_jnt}
      boxExtents=<10 10 10>
      constraintSetup {angularLimitsMax=<15 15 20>, angularLimitsMin=<-15 -15 -20>, angularTarget=<1 0 0>}
      localJointOffset=<1 0 0>
     }]
    simSpaceSettings {maxAngularAcceleration=25, maxAngularVelocity=5, simSpaceAngularAlpha=1}
   }
   n23 = Anim Dynamics - Bone: CatGlasses_jnt(ComponentPose=@n22, Alpha=1, ExternalForce="0,0,0") {
    angularDampingOverride=0.999
    bGravityOverrideInSimSpace=true
    bOverrideAngularDamping=true
    bUseGravityOverride=true
    simulationSpace=RootRelative
    boundBone {boneName=CatGlasses_jnt}
    componentAppliedLinearAccClamp=<3000 3000 3000>
    componentLinearAccScale=<0.3 0.3 0.3>
    componentLinearVelScale=<0.3 0.3 0.3>
    gravityOverride {z=-980}
    physicsBodyDefinitions [{
      boundBone {boneName=CatGlasses_jnt}
      boxExtents=<10 10 10>
      constraintSetup {angularLimitsMax=<0 0 0>, angularLimitsMin=<0 -20 0>, angularTarget=<1 0 0>}
      localJointOffset=<1 0 0>
     }]
    simSpaceSettings {maxAngularAcceleration=25, maxAngularVelocity=5, simSpaceAngularAlpha=1}
   }
   n24 = Anim Dynamics - Bone: BagStar_01_01_jnt(ComponentPose=@n23, Alpha=1, ExternalForce="0,0,0") {
    angularDampingOverride=0.999
    bOverrideAngularDamping=true
    bUseGravityOverride=true
    simulationSpace=RootRelative
    boundBone {boneName=BagStar_01_01_jnt}
    componentAppliedLinearAccClamp=<3000 3000 3000>
    componentLinearAccScale=<0.3 0.3 0.3>
    componentLinearVelScale=<0.3 0.3 0.3>
    gravityOverride {z=-980}
    physicsBodyDefinitions [{
      boundBone {boneName=BagStar_01_01_jnt}
      boxExtents=<10 10 10>
      constraintSetup {angularLimitsMax=<7 3 20>, angularLimitsMin=<-7 -20 -20>, angularTarget=<1 0 0>}
      localJointOffset=<1 0 0>
     }]
    simSpaceSettings {maxAngularAcceleration=25, maxAngularVelocity=5, simSpaceAngularAlpha=1}
   }
   n25 = Anim Dynamics - Bone: BagStar_01_02_jnt(ComponentPose=@n24, Alpha=1, ExternalForce="0,0,0") {
    angularDampingOverride=0.999
    bOverrideAngularDamping=true
    bUseGravityOverride=true
    simulationSpace=RootRelative
    boundBone {boneName=BagStar_01_02_jnt}
    componentAppliedLinearAccClamp=<3000 3000 3000>
    componentLinearAccScale=<0.3 0.3 0.3>
    componentLinearVelScale=<0.3 0.3 0.3>
    gravityOverride {z=-980}
    physicsBodyDefinitions [{
      boundBone {boneName=BagStar_01_02_jnt}
      boxExtents=<10 10 10>
      constraintSetup {angularLimitsMax=<7 3 20>, angularLimitsMin=<-7 -20 -20>, angularTarget=<1 0 0>}
      localJointOffset=<1 0 0>
     }]
    simSpaceSettings {maxAngularAcceleration=25, maxAngularVelocity=5, simSpaceAngularAlpha=1}
   }
   n26 = Anim Dynamics - Bone: Necklace_02_01_jnt(ComponentPose=@n25, Alpha=1, ExternalForce="0,0,0") {
    angularDampingOverride=0.9999
    angularSpringConstant=200
    bAngularSpring=true
    bOverrideAngularDamping=true
    bUseGravityOverride=true
    simulationSpace=RootRelative
    boundBone {boneName=Necklace_02_01_jnt}
    componentAppliedLinearAccClamp=<5000 5000 5000>
    componentLinearAccScale=<0.5 0.5 0.5>
    componentLinearVelScale=<0.5 0.5 0.5>
    gravityOverride {z=-980}
    physicsBodyDefinitions [{
      boundBone {boneName=Necklace_02_01_jnt}
      boxExtents=<10 10 10>
      constraintSetup {angularLimitsMax=<5 0 20>, angularLimitsMin=<-5 -20 -20>, angularTarget=<1 0 0>}
      localJointOffset=<0.7 0 0>
     }]
    simSpaceSettings {maxAngularAcceleration=25, maxAngularVelocity=5, simSpaceAngularAlpha=1}
   }
   n27 = Anim Dynamics - Bone: Necklace_02_02_jnt(ComponentPose=@n26, Alpha=1, ExternalForce="0,0,0") {
    angularDampingOverride=0.9999
    angularSpringConstant=200
    bAngularSpring=true
    bOverrideAngularDamping=true
    bUseGravityOverride=true
    simulationSpace=RootRelative
    boundBone {boneName=Necklace_02_02_jnt}
    componentAppliedLinearAccClamp=<5000 5000 5000>
    componentLinearAccScale=<0.5 0.5 0.5>
    componentLinearVelScale=<0.5 0.5 0.5>
    gravityOverride {z=-980}
    physicsBodyDefinitions [{
      boundBone {boneName=Necklace_02_02_jnt}
      boxExtents=<10 10 10>
      constraintSetup {angularLimitsMax=<5 0 20>, angularLimitsMin=<-5 -30 -20>, angularTarget=<1 0 0>}
      localJointOffset=<0.7 0 0>
     }]
    simSpaceSettings {maxAngularAcceleration=25, maxAngularVelocity=5, simSpaceAngularAlpha=1}
   }
   n28 = Anim Dynamics - Bone: Necklace_02_03_jnt(ComponentPose=@n27, Alpha=1, ExternalForce="0,0,0") {
    angularDampingOverride=0.9999
    angularSpringConstant=200
    bAngularSpring=true
    bOverrideAngularDamping=true
    bUseGravityOverride=true
    simulationSpace=RootRelative
    boundBone {boneName=Necklace_02_03_jnt}
    componentAppliedLinearAccClamp=<5000 5000 5000>
    componentLinearAccScale=<0.5 0.5 0.5>
    componentLinearVelScale=<0.5 0.5 0.5>
    gravityOverride {z=-980}
    physicsBodyDefinitions [{
      boundBone {boneName=Necklace_02_03_jnt}
      boxExtents=<10 10 10>
      constraintSetup {angularLimitsMax=<5 0 20>, angularLimitsMin=<-5 -30 -20>, angularTarget=<1 0 0>}
      localJointOffset=<0.7 0 0>
     }]
    simSpaceSettings {maxAngularAcceleration=25, maxAngularVelocity=5, simSpaceAngularAlpha=1}
   }
   n29 = "Constraint - Necklace_03_01_base_jnt (R)"(ComponentPose=@n28, Alpha=1, ConstraintWeights_0=1) {
    boneToModify {boneName=Necklace_03_01_base_jnt}
    constraintSetup [{transformType=Rotation, targetBone {boneName=Necklace_02_01_jnt}}]
    constraintWeights [1]
   }
   n30 = Anim Dynamics - Bone: Necklace_03_01_jnt(ComponentPose=@n29, Alpha=1, ExternalForce="0,0,0") {
    angularDampingOverride=0.999
    bOverrideAngularDamping=true
    bUseGravityOverride=true
    simulationSpace=RootRelative
    boundBone {boneName=Necklace_03_01_jnt}
    componentAppliedLinearAccClamp=<3000 3000 3000>
    componentLinearAccScale=<0.3 0.3 0.3>
    componentLinearVelScale=<0.3 0.3 0.3>
    gravityOverride {z=-980}
    physicsBodyDefinitions [{
      boundBone {boneName=Necklace_03_01_jnt}
      boxExtents=<10 10 10>
      constraintSetup {angularLimitsMax=<3 3 20>, angularLimitsMin=<-3 -20 -20>, angularTarget=<1 0 0>}
      localJointOffset=<1 0 0>
     }]
    simSpaceSettings {maxAngularAcceleration=25, maxAngularVelocity=5, simSpaceAngularAlpha=1}
   }
   n31 = Anim Dynamics - Bone: Necklace_08_01_jnt(ComponentPose=@n30, Alpha=1, ExternalForce="0,0,0") {
    angularDampingOverride=0.9999
    bOverrideAngularDamping=true
    bUseGravityOverride=true
    simulationSpace=RootRelative
    boundBone {boneName=Necklace_08_01_jnt}
    componentAppliedLinearAccClamp=<3000 3000 3000>
    componentLinearAccScale=<0.5 0.5 0.5>
    componentLinearVelScale=<0.5 0.5 0.5>
    gravityOverride {z=-980}
    physicsBodyDefinitions [{
      boundBone {boneName=Necklace_08_01_jnt}
      boxExtents=<10 10 10>
      constraintSetup {angularLimitsMax=<1 0 20>, angularLimitsMin=<-1 -20 -20>, angularTarget=<1 0 0>}
      localJointOffset=<1 0 0>
     }]
    simSpaceSettings {maxAngularAcceleration=25, maxAngularVelocity=5, simSpaceAngularAlpha=1}
   }
   n32 = Anim Dynamics - Bone: Necklace_08_02_jnt(ComponentPose=@n31, Alpha=1, ExternalForce="0,0,0") {
    angularDampingOverride=0.9999
    bOverrideAngularDamping=true
    bUseGravityOverride=true
    simulationSpace=RootRelative
    boundBone {boneName=Necklace_08_02_jnt}
    componentAppliedLinearAccClamp=<3000 3000 3000>
    componentLinearAccScale=<0.5 0.5 0.5>
    componentLinearVelScale=<0.5 0.5 0.5>
    gravityOverride {z=-980}
    physicsBodyDefinitions [{
      boundBone {boneName=Necklace_08_02_jnt}
      boxExtents=<10 10 10>
      constraintSetup {angularLimitsMax=<1 0 20>, angularLimitsMin=<-1 -20 -20>, angularTarget=<1 0 0>}
      localJointOffset=<1 0 0>
     }]
    simSpaceSettings {maxAngularAcceleration=25, maxAngularVelocity=5, simSpaceAngularAlpha=1}
   }
   n33 = Anim Dynamics - Bone: Necklace_08_03_jnt(ComponentPose=@n32, Alpha=1, ExternalForce="0,0,0") {
    angularDampingOverride=0.9999
    bOverrideAngularDamping=true
    bUseGravityOverride=true
    simulationSpace=RootRelative
    boundBone {boneName=Necklace_08_03_jnt}
    componentAppliedLinearAccClamp=<3000 3000 3000>
    componentLinearAccScale=<0.5 0.5 0.5>
    componentLinearVelScale=<0.5 0.5 0.5>
    gravityOverride {z=-980}
    physicsBodyDefinitions [{
      boundBone {boneName=Necklace_08_03_jnt}
      boxExtents=<10 10 10>
      constraintSetup {angularLimitsMax=<1 0 20>, angularLimitsMin=<-1 -20 -20>, angularTarget=<1 0 0>}
      localJointOffset=<1 0 0>
     }]
    simSpaceSettings {maxAngularAcceleration=25, maxAngularVelocity=5, simSpaceAngularAlpha=1}
   }
   n34 = Anim Dynamics - Bone: Necklace_04_01_jnt(ComponentPose=@n33, Alpha=1, ExternalForce="0,0,0") {
    angularDampingOverride=0.9999
    bOverrideAngularDamping=true
    bUseGravityOverride=true
    simulationSpace=RootRelative
    boundBone {boneName=Necklace_04_01_jnt}
    componentAppliedLinearAccClamp=<3000 3000 3000>
    componentLinearAccScale=<0.3 0.3 0.3>
    componentLinearVelScale=<0.3 0.3 0.3>
    gravityOverride {z=-980}
    physicsBodyDefinitions [{
      boundBone {boneName=Necklace_04_01_jnt}
      boxExtents=<9 9 9>
      constraintSetup {angularLimitsMax=<5 0 20>, angularLimitsMin=<-5 -15 -20>, angularTarget=<1 0 0>}
      localJointOffset=<1 0 0>
     }]
    simSpaceSettings {maxAngularAcceleration=25, maxAngularVelocity=5, simSpaceAngularAlpha=1}
   }
   n35 = Anim Dynamics - Bone: Necklace_04_02_jnt(ComponentPose=@n34, Alpha=1, ExternalForce="0,0,0") {
    angularDampingOverride=0.9999
    bOverrideAngularDamping=true
    bUseGravityOverride=true
    simulationSpace=RootRelative
    boundBone {boneName=Necklace_04_02_jnt}
    componentAppliedLinearAccClamp=<3000 3000 3000>
    componentLinearAccScale=<0.3 0.3 0.3>
    componentLinearVelScale=<0.3 0.3 0.3>
    gravityOverride {z=-980}
    physicsBodyDefinitions [{
      boundBone {boneName=Necklace_04_02_jnt}
      boxExtents=<9 9 9>
      constraintSetup {angularLimitsMax=<5 0 20>, angularLimitsMin=<-5 -15 -20>, angularTarget=<1 0 0>}
      localJointOffset=<1 0 0>
     }]
    simSpaceSettings {maxAngularAcceleration=25, maxAngularVelocity=5, simSpaceAngularAlpha=1}
   }
   n36 = Anim Dynamics - Bone: Glasses_jnt(ComponentPose=@n35, Alpha=1, ExternalForce="0,0,0") {
    angularDampingOverride=0.999
    bOverrideAngularDamping=true
    bUseGravityOverride=true
    simulationSpace=RootRelative
    boundBone {boneName=Glasses_jnt}
    componentAppliedLinearAccClamp=<3000 3000 3000>
    componentLinearAccScale=<0.3 0.3 0.3>
    componentLinearVelScale=<0.3 0.3 0.3>
    gravityOverride {z=-980}
    physicsBodyDefinitions [{
      boundBone {boneName=Glasses_jnt}
      boxExtents=<10 10 10>
      constraintSetup {angularLimitsMax=<3 3 3>, angularLimitsMin=<-3 -30 -3>, angularTarget=<1 0 0>}
      localJointOffset=<1 0 0>
     }]
    simSpaceSettings {maxAngularAcceleration=25, maxAngularVelocity=5, simSpaceAngularAlpha=1}
   }
   Output Pose(Result=Component To Local(ComponentPose=@n36)) {layerGroup=None}
   Get Stage4() -> (Stage4:double)
  ]}
 ChestOpeningState {nodes [ChestOpeningState!() -> (then, ReactivityState:int) {is=title K2Node_FunctionEntry}]}
 EventGraph {nodes [
   Event BlueprintUpdateAnimation()
   > n1 = Sequence!(execute) -> (then_0, then_1, then_2:Exec, then_3:Exec)
   n2 = Get IsAlwaysVisible() -> (IsAlwaysVisible:bool)
   > Branch(Condition=OR Boolean(A=@n2, B=Get IsDemo() -> (IsDemo:bool)))
   > Do Once(Reset=[@n3.else, @n4.then])
   > Set Visual States(Visual States:int=int + int(A=Get Visual States() -> (Visual States:int=0))) -> (Output_Get:int=0)
   > Delay(Duration=5)
   > n3 = Branch(Condition="Equal (Integer)"(A=Get Visual States() -> (Visual States=0), B=5))
   > n4 = Set Visual States(Visual States=0) -> (Output_Get=0)
   n5 = Branch(execute=@n1.then_1, Condition=Get IsAlwaysVisible())
   Set Stage2(execute=[@n7.then, @n5.then], Stage2:double=0) -> (Output_Get:double)
   > Set Stage3(Stage3:double=0) -> (Output_Get:double)
   > Set Stage4(Stage4:double=0) -> (Output_Get:double)
   > Set Stage5(Stage5:double=0) -> (Output_Get:double)
   > Set GoldenBarScale5(GoldenBarScale5:bool=false) -> (Output_Get:bool)
   n6 = Get Visual States() -> (Visual States=0)
   n7 = Branch(execute=@n5.else, Condition="Equal (Integer)"(A=@n6, B=4))
   n8 = Branch(execute=@n7.else, Condition="Equal (Integer)"(A=@n6, B=3))
   > Set Stage2(Stage2=1)
   > Set Stage3(Stage3=0)
   > Set Stage4(Stage4=0)
   > Set Stage5(Stage5=0)
   > Set GoldenBarScale4(GoldenBarScale4:bool=false) -> (Output_Get:bool)
   n9 = Branch(execute=@n8.else, Condition="Equal (Integer)"(A=@n6, B=2))
   > Set Stage2(Stage2=1)
   > Set Stage3(Stage3=1)
   > Set Stage4(Stage4=0)
   > Set Stage5(Stage5=0)
   > Set GoldenBarScale4(GoldenBarScale4=true)
   > Set GoldenBarScale5(GoldenBarScale5=true)
   n10 = Branch(execute=@n9.else, Condition="Equal (Integer)"(A=@n6, B=1))
   > Set Stage2(Stage2=1)
   > Set Stage3(Stage3=1)
   > Set Stage4(Stage4=1)
   > Set Stage5(Stage5=0)
   Set Stage2(execute=@n10.else, Stage2=1)
   > Set Stage3(Stage3=1)
   > Set Stage4(Stage4=1)
   > Set Stage5(Stage5=1)
   n11 = Get Dynamic Material Instance(execute=@n1.then_2, SkeletalMeshComponent=GetPartSkeletalMeshComponent())
   > SetScalarParameterValue(self=@n11, ParameterName=Hide Element 10, Value=Get Stage2() -> (Stage2:double))
   > SetScalarParameterValue(self=@n11, ParameterName=Hide Element 6_5, Value=Get Stage3() -> (Stage3:double))
   > SetScalarParameterValue(self=@n11, ParameterName=Hide Element 04, Value=Get Stage4())
   > SetScalarParameterValue(self=@n11, ParameterName=Hide Element 02, Value=Get Stage5() -> (Stage5:double))
   Call Chest Opening State!(execute:Exec=@n1.then_3, self:F MED Hammer Vice Pack Anim BP Object Reference, ReactivityState:int=Get Visual States() -> (Visual States=0)) -> (then:Exec) {is=title K2Node_CallDelegate}
   n12 = Event BlueprintProcessAnimInputEvent()
   > Branch(Condition=IsNamed(self=@n12.AnimInputEvent, InEventName=aie.ChestOpened))
   > Branch(Condition="integer < integer"(A=Get Visual States() -> (Visual States=0), B=4))
   > Set Visual States(Visual States=int + int(A=Get Visual States() -> (Visual States=0))) -> (Output_Get=0)
   Event OnCharacterCustomizationLoaded()
   > Set IsDemo(IsDemo:bool=OR Boolean(A=Get AnimTest() -> (AnimTest:bool), B=Get bIsInFrontEnd())) -> (Output_Get:bool)
   > Set Visual States(Visual States=0) -> (Output_Get=0)
   TryGetPawnOwner()
   NOT Boolean(A=@n2)
  ]}
]
variables [
 AnimTest {type=boolean}
 ChestOpeningState {default=()}
 GoldenBarScale4 {type=boolean}
 GoldenBarScale5 {type=boolean}
 IsAlwaysVisible {type=boolean}
 IsDemo {type=boolean}
 Stage2 {type=number, default=1}
 Stage3 {type=number, default=1}
 Stage4 {type=number, default=1}
 Stage5 {type=number, default=1}
 Visual States {type=integer}
]
